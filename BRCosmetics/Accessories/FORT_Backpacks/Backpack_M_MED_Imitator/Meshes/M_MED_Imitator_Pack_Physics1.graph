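# Physics asset for the Imitator backpack cosmetic, defining collision bodies and joints for its dynamic tentacle bones.
# Every body is a single sphere, with a large one on the pack bone and small ones on the tentacle bones.
# Constraints form left and right four-bone chains hanging from the pack.
# All joints share the same angular limit values, tight on twist and the first swing axis and wider on the second swing axis.
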PhysicsAsset /BRCosmetics/Accessories/FORT_Backpacks/Backpack_M_MED_Imitator/Meshes/M_MED_Imitator_Pack_Physics1
bodies {
 dyn_tentacles_C_1_l [{shapeType=Sphere, radius=2, shapeName=dyn_tentacles_C_1_l_sphere, center=<0.8681 3.8538 -0.0473>}]
 dyn_tentacles_C_1_r [{shapeType=Sphere, radius=2, shapeName=dyn_tentacles_C_1_r_sphere, center=<2.1309 3.326 0.0681>}]
 dyn_tentacles_C_2_l [{shapeType=Sphere, radius=2, shapeName=dyn_tentacles_C_2_l_sphere, center=<-0.8501 3.8575 -0.0673>}]
 dyn_tentacles_C_2_r [{shapeType=Sphere, radius=2, shapeName=dyn_tentacles_C_2_r_sphere, center=<1.8656 3.326 -1.032>}]
 dyn_tentacles_C_3_r [{shapeType=Sphere, radius=2, shapeName=dyn_tentacles_C_3_r_sphere, center=<0.0515 3.326 -2.1314>}]
 dyn_tentacles_down_1_l [{shapeType=Sphere, radius=2, shapeName=dyn_tentacles_down_1_l_sphere, center=<3.8416 0.6281 -0.6749>}]
 dyn_tentacles_down_2_l [{shapeType=Sphere, radius=2, shapeName=dyn_tentacles_down_2_l_sphere, center=<3.8416 0.6281 -0.6749>}]
 dyn_tentacles_down_r [{shapeType=Sphere, radius=2, shapeName=dyn_tentacles_down_r_sphere, center=<3.8416 0.6281 0.6749>}]
 pack [{shapeType=Sphere, radius=7.4086, shapeName=pack_sphere, center=<-0.3055 -1.5988 -0.0139>}]
}
constraints [
 {bone1Name=dyn_tentacles_C_3_r, bone2Name=dyn_tentacles_C_2_r, swing1LimitDegrees=5, swing2LimitDegrees=25, twistLimitDegrees=5}
 {bone1Name=dyn_tentacles_down_r, bone2Name=pack, swing1LimitDegrees=5, swing2LimitDegrees=25, twistLimitDegrees=5}
 {bone1Name=dyn_tentacles_C_1_r, bone2Name=dyn_tentacles_down_r, swing1LimitDegrees=5, swing2LimitDegrees=25, twistLimitDegrees=5}
 {bone1Name=dyn_tentacles_C_2_r, bone2Name=dyn_tentacles_C_1_r, swing1LimitDegrees=5, swing2LimitDegrees=25, twistLimitDegrees=5}
 {bone1Name=dyn_tentacles_C_2_l, bone2Name=dyn_tentacles_C_1_l, swing1LimitDegrees=5, swing2LimitDegrees=25, twistLimitDegrees=5}
 {bone1Name=dyn_tentacles_down_1_l, bone2Name=pack, swing1LimitDegrees=5, swing2LimitDegrees=25, twistLimitDegrees=5}
 {bone1Name=dyn_tentacles_down_2_l, bone2Name=dyn_tentacles_down_1_l, swing1LimitDegrees=5, swing2LimitDegrees=25, twistLimitDegrees=5}
 {bone1Name=dyn_tentacles_C_1_l, bone2Name=dyn_tentacles_down_2_l, swing1LimitDegrees=5, swing2LimitDegrees=25, twistLimitDegrees=5}
]
