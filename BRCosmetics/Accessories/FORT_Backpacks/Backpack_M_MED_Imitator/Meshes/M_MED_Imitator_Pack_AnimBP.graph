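# Animation Blueprint for the Imitator backpack cosmetic, animating its eye, eyelids, tentacles and swords.
# Update sets time-based sine rotations for eye sway, blinking and phase-shifted tentacle waves, applied through bone transforms after RigidBody physics.
# An equipped Imitator-tagged pickaxe outside skydiving toggles a material hide parameter, extra physics, tentacle offsets and springs.
# Parachute and skydive-float states blend pose offsets; cosmetic-stat changes trigger a brief kill pose, cleared after a delay.
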AnimBlueprint /BRCosmetics/Accessories/FORT_Backpacks/Backpack_M_MED_Imitator/Meshes/M_MED_Imitator_Pack_AnimBP
target_skeleton=&/M_MED_Imitator_Pack_Skeleton
graphs [
 AnimGraph {nodes [
   n1 = "Anim Dynamics - Chain: dyn_tentacles_down_r -> dyn_tentacles_C_1_r"!(ComponentPose, Alpha=1, GravityOverride:Vector="0,0,-980", ExternalForce="0,0,0") -> (Pose) {
    angularBiasOverride=0.2
    angularDampingOverride=0.2
    angularSpringConstant=0.2
    bAngularSpring=true
    bChain=true
    bOverrideAngularBias=true
    bOverrideAngularDamping=true
    bUseGravityOverride=true
    linearSpringConstant=0.1
    simulationSpace=BoneRelative
    alphaBoolBlend {blendOption=Linear}
    boundBone {boneName=dyn_tentacles_down_r}
    chainEnd {boneName=dyn_tentacles_C_1_r}
    componentLinearAccScale=<0.2 0.2 0.2>
    componentLinearVelScale=<0.2 0.2 0.2>
    gravityOverride {z=-980}
    physicsBodyDefinitions [
     {
      boundBone {boneName=dyn_tentacles_down_r}
      boxExtents=<1.25 3 3>
      constraintSetup {angularLimitsMax=<0 0 0>, angularLimitsMin=<0 0 0>, angularTarget=<0 0 0>}
      localJointOffset=<0 0 0>
     }
     {
      boundBone {boneName=dyn_tentacles_C_1_r}
      boxExtents=<1.25 3 2.5>
      constraintSetup {angularLimitsMax=<0 0 0>, angularLimitsMin=<0 0 0>, angularTarget=<0 0 0>}
      localJointOffset=<0 0 0>
     }
    ]
    relativeSpaceBone {boneName=dyn_sword_r}
   }
   "Anim Dynamics - Chain: dyn_tentacles_down_1_l -> dyn_tentacles_down_2_l"!(ComponentPose=@n1, Alpha=1, GravityOverride:Vector="0,0,-980", ExternalForce="0,0,0") -> (Pose) {
    angularBiasOverride=0.2
    angularDampingOverride=0.2
    angularSpringConstant=0.2
    bAngularSpring=true
    bChain=true
    bOverrideAngularBias=true
    bOverrideAngularDamping=true
    bUseGravityOverride=true
    linearSpringConstant=0.1
    simulationSpace=BoneRelative
    alphaBoolBlend {blendOption=Linear}
    boundBone {boneName=dyn_tentacles_down_1_l}
    chainEnd {boneName=dyn_tentacles_down_2_l}
    componentLinearAccScale=<0.2 0.2 0.2>
    componentLinearVelScale=<0.2 0.2 0.2>
    gravityOverride {z=-980}
    physicsBodyDefinitions [
     {
      boundBone {boneName=dyn_tentacles_down_1_l}
      boxExtents=<1 3 3>
      constraintSetup {angularLimitsMax=<0 0 0>, angularLimitsMin=<0 0 0>, angularTarget=<0 0 0>}
      localJointOffset=<0 0 0>
     }
     {
      boundBone {boneName=dyn_tentacles_down_2_l}
      boxExtents=<1 3 3>
      constraintSetup {angularLimitsMax=<0 0 0>, angularLimitsMin=<0 0 0>, angularTarget=<0 0 0>}
      localJointOffset=<0 0 0>
     }
    ]
    relativeSpaceBone {boneName=dyn_sword_l}
   }
   n2 = Get Weapon hands() -> (Weapon hands:bool)
   n3 = Get Active_Kill() -> (Active_Kill:bool)
   n4 = Get Rotate Blink Eye Up() -> (Rotate Blink Eye Up:double)
   n5 = Get Gravity Vector() -> (Gravity Vector:Vector)
   n6 = RigidBody!(ComponentPose=Local To Component(LocalPose=Input Pose()), bAlphaBoolEnabled:bool, OverrideWorldGravity:Vector=@n5, ExternalForce) -> (Pose) {
    alphaInputType=Bool
    bClampLinearTranslationLimitToRefPose=true
    bOverrideWorldGravity=true
    simulationSpace=BaseBoneSpace
    alphaBoolBlend {blendOption=Linear}
    overrideWorldGravity {z=-980}
   }
   n7 = RigidBody(ComponentPose=@n6, bAlphaBoolEnabled=Get Weapon hands(), OverrideWorldGravity=@n5) {
    alphaInputType=Bool
    bClampLinearTranslationLimitToRefPose=true
    bOverrideWorldGravity=true
    overridePhysicsAsset=&/M_MED_Imitator_Pack_Physics1
    simulationSpace=BaseBoneSpace
    alphaBoolBlend {blendOption=Linear}
    componentLinearAccScale=<0.2 0.2 0.2>
    componentLinearVelScale=<0.2 0.2 0.2>
    overrideWorldGravity {z=-980}
   }
   n8 = "Transform (Modify) Bone - Bone: pack"!(ComponentPose=@n7, bAlphaBoolEnabled:bool, Translation="0.000,-0.770,0.025", Rotation="-0.000,0.000,1.881", Scale) -> (Pose) {
    alphaInputType=Bool
    rotationMode=BMM_Additive
    rotationSpace=BCS_BoneSpace
    scaleMode=BMM_Additive
    scaleSpace=BCS_BoneSpace
    translationMode=BMM_Additive
    translationSpace=BCS_BoneSpace
    alphaBoolBlend {blendOption=Linear}
    boneToModify {boneName=pack}
    rotation {roll=1.881}
    translation {y=-0.77, z=0.025}
   }
   n9 = "Transform (Modify) Bone - Bone: dyn_eye_all"!(ComponentPose=@n8, bAlphaBoolEnabled:bool=True, Translation="0,0,0", Rotation=Get RotateBaseEye() -> (RotateBaseEye:Rotator), Scale="1,1,1") -> (Pose) {
    alphaInputType=Bool
    rotationMode=BMM_Additive
    rotationSpace=BCS_BoneSpace
    scaleMode=BMM_Additive
    scaleSpace=BCS_BoneSpace
    translationMode=BMM_Additive
    translationSpace=BCS_BoneSpace
    alphaBoolBlend {blendOption=Linear}
    boneToModify {boneName=dyn_eye_all}
   }
   n10 = Get Rotate Tentcl A() -> (Rotate Tentcl A:Rotator)
   n11 = "Transform (Modify) Bone - Bone: dyn_tentacles_A_1_r"!(ComponentPose=@n9, bAlphaBoolEnabled:bool=True, Translation="0,0,0", Rotation=@n10, Scale="1,1,1") -> (Pose) {
    alphaInputType=Bool
    rotationMode=BMM_Additive
    rotationSpace=BCS_BoneSpace
    scaleMode=BMM_Additive
    scaleSpace=BCS_BoneSpace
    translationMode=BMM_Additive
    translationSpace=BCS_BoneSpace
    alphaBoolBlend {blendOption=Linear}
    boneToModify {boneName=dyn_tentacles_A_1_r}
   }
   n12 = Get Rotate Tentcl B() -> (Rotate Tentcl B:Rotator)
   n13 = "Transform (Modify) Bone - Bone: dyn_tentacles_A_2_r"!(ComponentPose=@n11, bAlphaBoolEnabled:bool=True, Translation="0,0,0", Rotation=@n12, Scale="1,1,1") -> (Pose) {
    alphaInputType=Bool
    rotationMode=BMM_Additive
    rotationSpace=BCS_BoneSpace
    scaleMode=BMM_Additive
    scaleSpace=BCS_BoneSpace
    translationMode=BMM_Additive
    translationSpace=BCS_BoneSpace
    alphaBoolBlend {blendOption=Linear}
    boneToModify {boneName=dyn_tentacles_A_2_r}
   }
   n14 = "Transform (Modify) Bone - Bone: dyn_tentacles_C_1_r"!(ComponentPose=@n13, bAlphaBoolEnabled:bool=True, Translation="0,0,0", Rotation=@n10, Scale="1,1,1") -> (Pose) {
    alphaInputType=Bool
    rotationMode=BMM_Additive
    rotationSpace=BCS_BoneSpace
    scaleMode=BMM_Additive
    scaleSpace=BCS_BoneSpace
    translationMode=BMM_Additive
    translationSpace=BCS_BoneSpace
    alphaBoolBlend {blendOption=Linear}
    boneToModify {boneName=dyn_tentacles_C_1_r}
   }
   n15 = "Transform (Modify) Bone - Bone: dyn_tentacles_C_2_r"!(ComponentPose=@n14, bAlphaBoolEnabled:bool=True, Translation="0,0,0", Rotation=@n12, Scale="1,1,1") -> (Pose) {
    alphaInputType=Bool
    rotationMode=BMM_Additive
    rotationSpace=BCS_BoneSpace
    scaleMode=BMM_Additive
    scaleSpace=BCS_BoneSpace
    translationMode=BMM_Additive
    translationSpace=BCS_BoneSpace
    alphaBoolBlend {blendOption=Linear}
    boneToModify {boneName=dyn_tentacles_C_2_r}
   }
   n16 = Get Rotate Tentcl C() -> (Rotate Tentcl C:Rotator)
   n17 = "Transform (Modify) Bone - Bone: dyn_tentacles_C_3_r"!(ComponentPose=@n15, bAlphaBoolEnabled:bool=True, Translation="0,0,0", Rotation=@n16, Scale="1,1,1") -> (Pose) {
    alphaInputType=Bool
    rotationMode=BMM_Additive
    rotationSpace=BCS_BoneSpace
    scaleMode=BMM_Additive
    scaleSpace=BCS_BoneSpace
    translationMode=BMM_Additive
    translationSpace=BCS_BoneSpace
    alphaBoolBlend {blendOption=Linear}
    boneToModify {boneName=dyn_tentacles_C_3_r}
   }
   n18 = "Transform (Modify) Bone - Bone: dyn_tentacles_B_1_r"!(ComponentPose=@n17, bAlphaBoolEnabled:bool=True, Translation="0,0,0", Rotation=@n10, Scale="1,1,1") -> (Pose) {
    alphaInputType=Bool
    rotationMode=BMM_Additive
    rotationSpace=BCS_BoneSpace
    scaleMode=BMM_Additive
    scaleSpace=BCS_BoneSpace
    translationMode=BMM_Additive
    translationSpace=BCS_BoneSpace
    alphaBoolBlend {blendOption=Linear}
    boneToModify {boneName=dyn_tentacles_B_1_r}
   }
   n19 = "Transform (Modify) Bone - Bone: dyn_tentacles_B_2_r"!(ComponentPose=@n18, bAlphaBoolEnabled:bool=True, Translation="0,0,0", Rotation=@n12, Scale="1,1,1") -> (Pose) {
    alphaInputType=Bool
    rotationMode=BMM_Additive
    rotationSpace=BCS_BoneSpace
    scaleMode=BMM_Additive
    scaleSpace=BCS_BoneSpace
    translationMode=BMM_Additive
    translationSpace=BCS_BoneSpace
    alphaBoolBlend {blendOption=Linear}
    boneToModify {boneName=dyn_tentacles_B_2_r}
   }
   n20 = "Transform (Modify) Bone - Bone: dyn_tentacles_B_3_r"!(ComponentPose=@n19, bAlphaBoolEnabled:bool=True, Translation="0,0,0", Rotation=@n16, Scale="1,1,1") -> (Pose) {
    alphaInputType=Bool
    rotationMode=BMM_Additive
    rotationSpace=BCS_BoneSpace
    scaleMode=BMM_Additive
    scaleSpace=BCS_BoneSpace
    translationMode=BMM_Additive
    translationSpace=BCS_BoneSpace
    alphaBoolBlend {blendOption=Linear}
    boneToModify {boneName=dyn_tentacles_B_3_r}
   }
   n21 = Get bIsParachuteOpened()
   n22 = "Transform (Modify) Bone - Bone: dyn_sword_l"!(ComponentPose=@n20, bAlphaBoolEnabled:bool=@n21, Translation="-0.594,-2.495,0.679", Rotation="-0.000,-13.373,-0.001", Scale="1,1,1") -> (Pose) {
    alphaInputType=Bool
    rotationMode=BMM_Additive
    rotationSpace=BCS_BoneSpace
    scaleMode=BMM_Additive
    scaleSpace=BCS_BoneSpace
    translationMode=BMM_Additive
    translationSpace=BCS_BoneSpace
    alphaBoolBlend {blendInTime=0.2, blendOption=Linear, blendOutTime=0.5}
    boneToModify {boneName=dyn_sword_l}
    rotation {roll=-0.001, yaw=-13.373}
    translation=<-0.594 -2.495 0.679>
   }
   n23 = "Transform (Modify) Bone - Bone: dyn_sword_r"!(ComponentPose=@n22, bAlphaBoolEnabled:bool=@n21, Translation="0,0,0", Rotation="-0.000,-13.373,-0.001", Scale="1,1,1") -> (Pose) {
    alphaInputType=Bool
    rotationMode=BMM_Additive
    rotationSpace=BCS_BoneSpace
    scaleMode=BMM_Additive
    scaleSpace=BCS_BoneSpace
    translationMode=BMM_Additive
    translationSpace=BCS_BoneSpace
    alphaBoolBlend {blendInTime=0.2, blendOption=Linear, blendOutTime=0.5}
    boneToModify {boneName=dyn_sword_r}
    rotation {roll=-0.001, yaw=-13.373}
   }
   n24 = "Transform (Modify) Bone - Bone: dyn_eyelid_down"(ComponentPose=@n23, Alpha=@n4, Translation="0,0,0", Rotation="0.004,-30.654,0.008", Scale="1,1,1") {
    rotationMode=BMM_Additive
    rotationSpace=BCS_BoneSpace
    scaleMode=BMM_Additive
    scaleSpace=BCS_BoneSpace
    translationMode=BMM_Additive
    translationSpace=BCS_BoneSpace
    alphaBoolBlend {blendOption=Linear}
    boneToModify {boneName=dyn_eyelid_down}
    rotation=<rot 0.004 -30.6539 0.008>
   }
   n25 = "Transform (Modify) Bone - Bone: dyn_eyelid_up"(ComponentPose=@n24, Alpha=@n4, Translation="0,0,0", Rotation="-0.000,27.620,0.005", Scale="1,1,1") {
    rotationMode=BMM_Additive
    rotationSpace=BCS_BoneSpace
    scaleMode=BMM_Additive
    scaleSpace=BCS_BoneSpace
    translationMode=BMM_Additive
    translationSpace=BCS_BoneSpace
    alphaBoolBlend {blendOption=Linear}
    boneToModify {boneName=dyn_eyelid_up}
    rotation=<rot 0 27.6201 0.005>
   }
   n26 = "Transform (Modify) Bone - Bone: dyn_tentacles_B_l"!(ComponentPose=@n25, bAlphaBoolEnabled:bool=True, Translation="0,0,0", Rotation=@n10, Scale="1,1,1") -> (Pose) {
    alphaInputType=Bool
    rotationMode=BMM_Additive
    rotationSpace=BCS_BoneSpace
    scaleMode=BMM_Additive
    scaleSpace=BCS_BoneSpace
    translationMode=BMM_Additive
    translationSpace=BCS_BoneSpace
    alphaBoolBlend {blendOption=Linear}
    boneToModify {boneName=dyn_tentacles_B_l}
   }
   n27 = "Transform (Modify) Bone - Bone: dyn_tentacles_C_1_l"!(ComponentPose=@n26, bAlphaBoolEnabled:bool=True, Translation="0,0,0", Rotation=@n16, Scale="1,1,1") -> (Pose) {
    alphaInputType=Bool
    rotationMode=BMM_Additive
    rotationSpace=BCS_BoneSpace
    scaleMode=BMM_Additive
    scaleSpace=BCS_BoneSpace
    translationMode=BMM_Additive
    translationSpace=BCS_BoneSpace
    alphaBoolBlend {blendOption=Linear}
    boneToModify {boneName=dyn_tentacles_C_1_l}
   }
   n28 = "Transform (Modify) Bone - Bone: dyn_tentacles_C_2_l"!(ComponentPose=@n27, bAlphaBoolEnabled:bool=True, Translation="0,0,0", Rotation=@n12, Scale="1,1,1") -> (Pose) {
    alphaInputType=Bool
    rotationMode=BMM_Additive
    rotationSpace=BCS_BoneSpace
    scaleMode=BMM_Additive
    scaleSpace=BCS_BoneSpace
    translationMode=BMM_Additive
    translationSpace=BCS_BoneSpace
    alphaBoolBlend {blendOption=Linear}
    boneToModify {boneName=dyn_tentacles_C_2_l}
   }
   n29 = "Transform (Modify) Bone - Bone: dyn_tentacles_A_1_l"!(ComponentPose=@n28, bAlphaBoolEnabled:bool=True, Translation="0,0,0", Rotation=@n16, Scale="1,1,1") -> (Pose) {
    alphaInputType=Bool
    rotationMode=BMM_Additive
    rotationSpace=BCS_BoneSpace
    scaleMode=BMM_Additive
    scaleSpace=BCS_BoneSpace
    translationMode=BMM_Additive
    translationSpace=BCS_BoneSpace
    alphaBoolBlend {blendOption=Linear}
    boneToModify {boneName=dyn_tentacles_A_1_l}
   }
   n30 = "Transform (Modify) Bone - Bone: dyn_tentacles_A_2_l"!(ComponentPose=@n29, bAlphaBoolEnabled:bool=True, Translation="0,0,0", Rotation=@n12, Scale="1,1,1") -> (Pose) {
    alphaInputType=Bool
    rotationMode=BMM_Additive
    rotationSpace=BCS_BoneSpace
    scaleMode=BMM_Additive
    scaleSpace=BCS_BoneSpace
    translationMode=BMM_Additive
    translationSpace=BCS_BoneSpace
    alphaBoolBlend {blendOption=Linear}
    boneToModify {boneName=dyn_tentacles_A_2_l}
   }
   n31 = Get bIsSkydiveFloating()
   n32 = "Transform (Modify) Bone - Bone: dyn_sword_r"(ComponentPose=@n30, bAlphaBoolEnabled=@n31, Translation="1.897,-5.416,0.050", Rotation="1.892,-9.641,-0.320", Scale="1,1,1") {
    alphaInputType=Bool
    rotationMode=BMM_Additive
    rotationSpace=BCS_BoneSpace
    scaleMode=BMM_Additive
    scaleSpace=BCS_BoneSpace
    translationMode=BMM_Additive
    translationSpace=BCS_BoneSpace
    alphaBoolBlend {blendInTime=1, blendOption=Linear, blendOutTime=1}
    boneToModify {boneName=dyn_sword_r}
    rotation=<rot 1.892 -9.6409 -0.32>
    translation=<1.897 -5.416 0.0496>
   }
   n33 = "Transform (Modify) Bone - Bone: dyn_sword_l"(ComponentPose=@n32, bAlphaBoolEnabled=@n31, Translation="2.602,-6.063,-0.2", Rotation="-5.885,-5.612,0.576", Scale="1,1,1") {
    alphaInputType=Bool
    rotationMode=BMM_Additive
    rotationSpace=BCS_BoneSpace
    scaleMode=BMM_Additive
    scaleSpace=BCS_BoneSpace
    translationMode=BMM_Additive
    translationSpace=BCS_BoneSpace
    alphaBoolBlend {blendInTime=1, blendOption=Linear, blendOutTime=1}
    boneToModify {boneName=dyn_sword_l}
    rotation=<rot -5.885 -5.612 0.576>
    translation=<2.602 -6.063 -0.2>
   }
   n34 = "Transform (Modify) Bone - Bone: dyn_head"!(ComponentPose=@n33, bAlphaBoolEnabled:bool=@n31, Translation="0,0,0", Rotation="0,0,0", Scale="1,1,1") -> (Pose) {
    alphaInputType=Bool
    rotationMode=BMM_Replace
    rotationSpace=BCS_BoneSpace
    scaleMode=BMM_Additive
    scaleSpace=BCS_BoneSpace
    translationMode=BMM_Replace
    translationSpace=BCS_BoneSpace
    alphaBoolBlend {blendInTime=1, blendOption=Linear, blendOutTime=1}
    boneToModify {boneName=dyn_head}
   }
   n35 = "Transform (Modify) Bone - Bone: dyn_eyelid_up"!(ComponentPose=@n34, bAlphaBoolEnabled:bool=@n3, Translation="0,0,0", Rotation="0.000,26.439,0.000", Scale="1,1,1") -> (Pose) {
    alphaInputType=Bool
    rotationMode=BMM_Replace
    rotationSpace=BCS_BoneSpace
    scaleMode=BMM_Replace
    scaleSpace=BCS_BoneSpace
    translationMode=BMM_Replace
    translationSpace=BCS_BoneSpace
    alphaBoolBlend {blendOption=Linear}
    boneToModify {boneName=dyn_eyelid_up}
    rotation {yaw=26.439}
   }
   n36 = "Transform (Modify) Bone - Bone: dyn_eyelid_down"!(ComponentPose=@n35, bAlphaBoolEnabled:bool=@n3, Translation="0,0,0", Rotation="0.000,-30.776,0.000", Scale="1,1,1") -> (Pose) {
    alphaInputType=Bool
    rotationMode=BMM_Replace
    rotationSpace=BCS_BoneSpace
    scaleMode=BMM_Replace
    scaleSpace=BCS_BoneSpace
    translationMode=BMM_Replace
    translationSpace=BCS_BoneSpace
    alphaBoolBlend {blendOption=Linear}
    boneToModify {boneName=dyn_eyelid_down}
    rotation=<rot 0 -30.776 0>
   }
   n37 = "Transform (Modify) Bone - Bone: dyn_head"(ComponentPose=@n36, bAlphaBoolEnabled=@n3, Translation="0.611,-0.449,0.000", Rotation="-0.000,-3.544,0.000", Scale="1,1,1") {
    alphaInputType=Bool
    rotationMode=BMM_Replace
    rotationSpace=BCS_BoneSpace
    scaleMode=BMM_Replace
    scaleSpace=BCS_BoneSpace
    translationMode=BMM_Replace
    translationSpace=BCS_BoneSpace
    alphaBoolBlend {blendOption=Linear}
    boneToModify {boneName=dyn_head}
    rotation {yaw=-3.544}
    translation=<0.611 -0.449>
   }
   n38 = "Transform (Modify) Bone - Bone: dyn_sword_r"(ComponentPose=@n37, bAlphaBoolEnabled=@n3, Translation="-0.858,-1.966,-0.001", Rotation="5.922,-6.978,-0.723", Scale="1,1,1") {
    alphaInputType=Bool
    rotationMode=BMM_Replace
    rotationSpace=BCS_BoneSpace
    scaleMode=BMM_Replace
    scaleSpace=BCS_BoneSpace
    translationMode=BMM_Replace
    translationSpace=BCS_BoneSpace
    alphaBoolBlend {blendOption=Linear}
    boneToModify {boneName=dyn_sword_r}
    rotation=<rot 5.922 -6.978 -0.723>
    translation=<-0.8575 -1.9661 -0.001>
   }
   n39 = "Transform (Modify) Bone - Bone: dyn_sword_l"(ComponentPose=@n38, bAlphaBoolEnabled=@n3, Translation="0.225,-0.891,-0.024", Rotation="-9.415,-6.238,1.024", Scale="1,1,1") {
    alphaInputType=Bool
    rotationMode=BMM_Replace
    rotationSpace=BCS_BoneSpace
    scaleMode=BMM_Replace
    scaleSpace=BCS_BoneSpace
    translationMode=BMM_Replace
    translationSpace=BCS_BoneSpace
    alphaBoolBlend {blendOption=Linear}
    boneToModify {boneName=dyn_sword_l}
    rotation=<rot -9.415 -6.238 1.024>
    translation=<0.225 -0.891 -0.024>
   }
   n40 = "Transform (Modify) Bone - Bone: dyn_tentacles_up_r"!(ComponentPose=@n39, bAlphaBoolEnabled:bool=@n2, Translation="5.106,0.184,-2.268", Rotation="33.184,10.816,3.889", Scale="1,1,1") -> (Pose) {
    alphaInputType=Bool
    rotationMode=BMM_Additive
    rotationSpace=BCS_BoneSpace
    scaleMode=BMM_Additive
    scaleSpace=BCS_BoneSpace
    translationMode=BMM_Additive
    translationSpace=BCS_BoneSpace
    alphaBoolBlend {blendOption=Linear}
    boneToModify {boneName=dyn_tentacles_up_r}
    rotation=<rot 33.184 10.816 3.889>
    translation=<5.106 0.184 -2.268>
   }
   n41 = "Transform (Modify) Bone - Bone: dyn_tentacles_down_r"!(ComponentPose=@n40, bAlphaBoolEnabled:bool=@n2, Translation="6.396,-4.676,4.470", Rotation="-38.582,-23.889,8.475", Scale="1,1,1") -> (Pose) {
    alphaInputType=Bool
    rotationMode=BMM_Additive
    rotationSpace=BCS_BoneSpace
    scaleMode=BMM_Additive
    scaleSpace=BCS_BoneSpace
    translationMode=BMM_Additive
    translationSpace=BCS_BoneSpace
    alphaBoolBlend {blendOption=Linear}
    boneToModify {boneName=dyn_tentacles_down_r}
    rotation=<rot -38.582 -23.889 8.475>
    translation=<6.3961 -4.6763 4.47>
   }
   n42 = "Transform (Modify) Bone - Bone: dyn_tentacles_B_1_r"(ComponentPose=@n41, bAlphaBoolEnabled=@n2, Translation="-0.013,1.632,2.820", Rotation="0.277,-0.025,-46.955", Scale="1,1,1") {
    alphaInputType=Bool
    rotationMode=BMM_Additive
    rotationSpace=BCS_BoneSpace
    scaleMode=BMM_Additive
    scaleSpace=BCS_BoneSpace
    translationMode=BMM_Additive
    translationSpace=BCS_BoneSpace
    alphaBoolBlend {blendOption=Linear}
    boneToModify {boneName=dyn_tentacles_B_1_r}
    rotation=<rot 0.277 -0.025 -46.955>
    translation=<-0.013 1.632 2.82>
   }
   n43 = "Transform (Modify) Bone - Bone: dyn_tentacles_B_l"(ComponentPose=@n42, bAlphaBoolEnabled=True, Translation="0,1,0", Rotation="0,0,0", Scale="1,1,1") {
    alphaInputType=Bool
    rotationMode=BMM_Additive
    rotationSpace=BCS_BoneSpace
    scaleMode=BMM_Additive
    scaleSpace=BCS_BoneSpace
    translationMode=BMM_Additive
    translationSpace=BCS_BoneSpace
    alphaBoolBlend {blendOption=Linear}
    boneToModify {boneName=dyn_tentacles_B_l}
    translation {y=1}
   }
   n44 = "Transform (Modify) Bone - Bone: dyn_tentacles_down_1_l"!(ComponentPose=@n43, bAlphaBoolEnabled:bool=@n2, Translation="7.048,2.899,-2.132", Rotation="8.981,-48.965,-2.852", Scale="1,1,1") -> (Pose) {
    alphaInputType=Bool
    rotationMode=BMM_Additive
    rotationSpace=BCS_BoneSpace
    scaleMode=BMM_Additive
    scaleSpace=BCS_BoneSpace
    translationMode=BMM_Additive
    translationSpace=BCS_BoneSpace
    alphaBoolBlend {blendOption=Linear}
    boneToModify {boneName=dyn_tentacles_down_1_l}
    rotation=<rot 8.981 -48.965 -2.852>
    translation=<7.048 2.899 -2.132>
   }
   n45 = "Transform (Modify) Bone - Bone: dyn_tentacles_down_2_l"!(ComponentPose=@n44, bAlphaBoolEnabled:bool=@n2, Translation="0.077,0.268,-2.967", Rotation="-14.927,-28.003,7.800", Scale="1,1,1") -> (Pose) {
    alphaInputType=Bool
    rotationMode=BMM_Additive
    rotationSpace=BCS_BoneSpace
    scaleMode=BMM_Additive
    scaleSpace=BCS_BoneSpace
    translationMode=BMM_Additive
    translationSpace=BCS_BoneSpace
    alphaBoolBlend {blendOption=Linear}
    boneToModify {boneName=dyn_tentacles_down_2_l}
    rotation=<rot -14.927 -28.003 7.8>
    translation=<0.077 0.268 -2.967>
   }
   n46 = "Transform (Modify) Bone - Bone: dyn_tentacles_up_l"!(ComponentPose=@n45, bAlphaBoolEnabled:bool=@n2, Translation="5.061,0.355,6.222", Rotation="-68.336,12.298,11.669", Scale="1,1,1") -> (Pose) {
    alphaInputType=Bool
    rotationMode=BMM_Additive
    rotationSpace=BCS_BoneSpace
    scaleMode=BMM_Additive
    scaleSpace=BCS_BoneSpace
    translationMode=BMM_Additive
    translationSpace=BCS_BoneSpace
    alphaBoolBlend {blendOption=Linear}
    boneToModify {boneName=dyn_tentacles_up_l}
    rotation=<rot -68.336 12.298 11.669>
    translation=<5.061 0.355 6.222>
   }
   n47 = Spring controller - Bone: dyn_tentacles_down_1_l!(ComponentPose:Component Space Pose Link Structure=@n46, Alpha:float=0.03) -> (Pose:Component Space Pose Link Structure) {
    is=title AnimGraphNode_SpringBone
    bLimitDisplacement=true
    bTranslateX=false
    bTranslateY=false
    maxDisplacement=40
    springDamping=5
    springStiffness=90
    alphaBoolBlend {blendOption=Linear}
    springBone {boneName=dyn_tentacles_down_1_l}
   }
   n48 = Spring controller - Bone: dyn_tentacles_down_2_l!(ComponentPose:Component Space Pose Link Structure=@n47, Alpha:float=0.03) -> (Pose:Component Space Pose Link Structure) {
    is=title AnimGraphNode_SpringBone
    bLimitDisplacement=true
    bTranslateX=false
    bTranslateY=false
    maxDisplacement=40
    springDamping=5
    springStiffness=90
    alphaBoolBlend {blendOption=Linear}
    springBone {boneName=dyn_tentacles_down_2_l}
   }
   n49 = Spring controller - Bone: dyn_tentacles_up_l!(ComponentPose:Component Space Pose Link Structure=@n48, Alpha:float=0.03) -> (Pose:Component Space Pose Link Structure) {
    is=title AnimGraphNode_SpringBone
    bLimitDisplacement=true
    bTranslateX=false
    bTranslateY=false
    maxDisplacement=40
    springDamping=5
    springStiffness=90
    alphaBoolBlend {blendOption=Linear}
    springBone {boneName=dyn_tentacles_up_l}
   }
   n50 = Spring controller - Bone: dyn_tentacles_up_r!(ComponentPose:Component Space Pose Link Structure=@n49, Alpha:float=0.03) -> (Pose:Component Space Pose Link Structure) {
    is=title AnimGraphNode_SpringBone
    bLimitDisplacement=true
    bTranslateX=false
    bTranslateY=false
    maxDisplacement=40
    springDamping=5
    springStiffness=90
    alphaBoolBlend {blendOption=Linear}
    springBone {boneName=dyn_tentacles_up_r}
   }
   n51 = Spring controller - Bone: dyn_tentacles_down_r!(ComponentPose:Component Space Pose Link Structure=@n50, Alpha:float=0.03) -> (Pose:Component Space Pose Link Structure) {
    is=title AnimGraphNode_SpringBone
    bLimitDisplacement=true
    bTranslateX=false
    bTranslateY=false
    maxDisplacement=40
    springDamping=5
    springStiffness=90
    alphaBoolBlend {blendOption=Linear}
    springBone {boneName=dyn_tentacles_down_r}
   }
   n52 = "Clamp (Float)"(Value="To Float (Boolean)"(InBool=Get Weapon hands()), Min=0, Max=0.06)
   n53 = Spring controller - Bone: dyn_tentacles_down_1_l(ComponentPose=@n51, Alpha=@n52) {
    bLimitDisplacement=true
    bRotateX=true
    bRotateY=true
    bRotateZ=true
    maxDisplacement=40
    springDamping=5
    springStiffness=90
    alphaBoolBlend {blendOption=Linear}
    springBone {boneName=dyn_tentacles_down_1_l}
   }
   n54 = Spring controller - Bone: dyn_tentacles_down_2_l(ComponentPose=@n53, Alpha=@n52) {
    bLimitDisplacement=true
    bRotateX=true
    bRotateY=true
    bRotateZ=true
    maxDisplacement=40
    springDamping=5
    springStiffness=90
    alphaBoolBlend {blendOption=Linear}
    springBone {boneName=dyn_tentacles_down_2_l}
   }
   n55 = Spring controller - Bone: dyn_tentacles_up_l(ComponentPose=@n54, Alpha=@n52) {
    bLimitDisplacement=true
    bRotateX=true
    bRotateY=true
    bRotateZ=true
    maxDisplacement=40
    springDamping=5
    springStiffness=90
    alphaBoolBlend {blendOption=Linear}
    springBone {boneName=dyn_tentacles_up_l}
   }
   n56 = Spring controller - Bone: dyn_tentacles_up_r(ComponentPose=@n55, Alpha=@n52) {
    bLimitDisplacement=true
    bRotateX=true
    bRotateY=true
    bRotateZ=true
    maxDisplacement=40
    springDamping=5
    springStiffness=90
    alphaBoolBlend {blendOption=Linear}
    springBone {boneName=dyn_tentacles_up_r}
   }
   n57 = Spring controller - Bone: dyn_tentacles_down_r(ComponentPose=@n56, Alpha=@n52) {
    bLimitDisplacement=true
    bRotateX=true
    bRotateY=true
    bRotateZ=true
    maxDisplacement=40
    springDamping=5
    springStiffness=90
    alphaBoolBlend {blendOption=Linear}
    springBone {boneName=dyn_tentacles_down_r}
   }
   n58 = "Transform (Modify) Bone - Bone: pack"!(ComponentPose=@n57, Alpha:float, Translation="0,-3,0", Rotation="0,0,0", Scale) -> (Pose) {
    translationMode=BMM_Additive
    translationSpace=BCS_BoneSpace
    alphaBoolBlend {blendOption=Linear}
    boneToModify {boneName=pack}
    translation {y=-3}
   }
   Output Pose(Result=Component To Local(ComponentPose=@n58)) {layerGroup=None}
  ]}
 EventGraph {nodes [
   Event BlueprintInitializeAnimation()
   > Set Skel_Mesh(Skel_Mesh=GetOwnerSkeletalMeshComponent())
   n1 = Event OnWeaponChanged()
   n2 = Get Dynamic Material Instance(execute=[@n1.then, @n11.then], SkeletalMeshComponent=GetPartSkeletalMeshComponent())
   > n3 = Sequence()
   n4 = PickaxeItemDefinitionHasAnyMetaTag(Tags=Make Array("[0]"=(TagName="Cosmetics.MetaTags.Imitator")), bRequiresEquipped=true)
   > n5 = Branch(Condition=@n4)
   > n6 = Branch(Condition=Get bIsSkyDiving())
   SetScalarParameterValue(execute=[@n8.then, @n6.then, @n5.else], self=@n2, ParameterName=Hide Element 01, Value=0)
   > Set Weapon hands(Weapon hands:bool=false) -> (Output_Get:bool)
   SetScalarParameterValue(execute=@n6.else, self=@n2, ParameterName=Hide Element 01, Value=1)
   > Set Weapon hands(Weapon hands=true)
   Branch(execute=[@n7.then, @n3.then_1], Condition=@n4)
   n7 = Event OnEndSkydiving()
   n8 = Event OnBeginSkydiving()
   Event OnCharacterCustomizationLoaded()
   > n9 = Cast To FortPlayerPawn(Object=TryGetFortPawn())
   > n10 = WatchCosmeticStat!(
    execute:Exec
    Pawn:Fort Player Pawn Object Reference=@n9
    StatTag:Fort Stat Manager Tag Structure
    DefaultValue:int=0
    bBroadcastInitialValue:bool=true
    InitialValueBroadcastDelay:float=-1
    ChangeValueBroadcastDelay:float=-1
    bConsumeWaitingNotifications:bool=true
   ) -> (then:Exec, OnStatChanged:Exec, StatName:Name, StatValue:int) {is=title K2Node_AsyncAction}
   Set KillCount(execute=@n10.OnStatChanged, KillCount:double="To Float (Integer)"(InInt=@n10.StatValue)) -> (Output_Get:double)
   > n11 = Set Active_Kill(Active_Kill:bool=true) -> (Output_Get:bool)
   Event BlueprintUpdateAnimation()
   > n12 = Sequence!(execute) -> (then_0, then_1, then_2:Exec)
   n13 = Gravity Override State Function DEPRECATED(
    execute=@n12.then_1
    CharacterPartAnimInstance=Self-Reference()
    GameTimeInSeconds=GetGameTimeInSeconds()
    GravityOverrideParams=Get Gravity override Params() -> (Gravity override Params:Gravity Override Params Struct Structure)
   )
   > Set Gravity Vector(Gravity Vector:Vector=@n13) -> (Output_Get:Vector)
   Get!(execute:Exec=@n12.then_2) -> (then:Exec, else:Exec, Skel_Mesh:Skeletal Mesh Component Object Reference)
   > Branch(Condition=Get Active_Kill())
   > PrintString!(execute, InString=- - - Kill - - -, bPrintToScreen, bPrintToLog, TextColor, Duration, Key) -> (then)
   > Delay(Duration=0.5)
   > Set Active_Kill(Active_Kill=false)
   n14 = GetGameTimeInSeconds()
   n15 = float * float(A=@n14, B=0.5)
   n16 = "Atan (Radians)"!(A:double=float * float(A="Sin (Radians)"(A=@n15), B=15)) -> (ReturnValue:double)
   Set RotateBaseEye(execute=@n12.then_0, RotateBaseEye:Rotator=MakeRotator(Pitch=float * float(A=@n16, B=10))) -> (Output_Get:Rotator)
   n17 = "Sin (Radians)"(A=float * float(A=@n14, B=1))
   > Set Rotate Tentcl A(Rotate Tentcl A:Rotator=MakeRotator(Yaw=float * float(A=@n17, B=15))) -> (Output_Get:Rotator)
   n18 = "Sin (Radians)"(A=float * float(A=float + float(A=@n14, B=-1), B=1))
   > Set Rotate Tentcl B(Rotate Tentcl B:Rotator=MakeRotator(Yaw=float * float(A=@n18, B=15))) -> (Output_Get:Rotator)
   n19 = "Sin (Radians)"(A=float * float(A=float + float(A=@n14, B=-2), B=1))
   > Set Rotate Tentcl C(Rotate Tentcl C:Rotator=MakeRotator(Yaw=float * float(A=@n19, B=20))) -> (Output_Get:Rotator)
   n20 = float * float(A=@n14, B=0.35)
   n21 = "Atan (Radians)"(A=float * float(A="Sin (Radians)"(A=@n20), B=30))
   n22 = float * float(A=@n21, B=@n21)
   > Set Rotate Blink Eye Up(Rotate Blink Eye Up:double=float + float(A=float * float(A=@n22, B=-0.4), B=1)) -> (Output_Get:double)
   TryGetPawnOwner()
   NormalizeAxis!(Angle:float=0) -> (ReturnValue:float)
  ]}
]
variables [
 Active_Kill {type=boolean}
 Gravity override Params {type=GravityOverrideParamsStruct}
 Gravity Vector {type=Vector}
 KillCount {type=number}
 Rotate Blink Eye Up {type=number}
 Rotate Tentcl A {type=Rotator}
 Rotate Tentcl B {type=Rotator}
 Rotate Tentcl C {type=Rotator}
 RotateBaseEye {type=Rotator}
 Skel_Mesh {type=SkeletalMeshComponent}
 Weapon hands {type=boolean}
]
