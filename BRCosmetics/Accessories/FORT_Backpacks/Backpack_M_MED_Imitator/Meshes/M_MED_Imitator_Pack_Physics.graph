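# Physics asset for the M_MED_Imitator back bling (backpack) cosmetic, defining physics bodies for its pack, head and two sword pieces.
# A box covers the pack, a capsule the dyn_head bone, and a box each of the left and right sword bones.
# All three dyn_ bones constrain directly to the pack, with matching tight angular limits on the swords and a wider primary swing but locked twist on the head.
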PhysicsAsset /BRCosmetics/Accessories/FORT_Backpacks/Backpack_M_MED_Imitator/Meshes/M_MED_Imitator_Pack_Physics
bodies {
 dyn_head [{shapeType=Capsule, length=5.8882, radius=3, shapeName=dyn_head_capsule, center=<0 3.0185 0>, rotation=<rot 0 0 90>}]
 dyn_sword_l [{
   shapeType=Box
   extentX=3.5
   extentY=5
   extentZ=15
   shapeName=dyn_sword_l_box
   center=<3.6782 0.4983 -0.8501>
   rotation=<rot 1.2942 95.9645 75.9329>
  }]
 dyn_sword_r [{
   shapeType=Box
   extentX=3.5
   extentY=5
   extentZ=15
   shapeName=dyn_sword_r_box
   center=<5.1239 1.0257 0.7046>
   rotation=<rot -1.2227 99.3717 98.5118>
  }]
 pack [{shapeType=Box, extentX=11.9339, extentY=36.8829, extentZ=22.3171, shapeName=pack_box, center=<0 6.5174 -13.1078>, rotation=<rot 0 90 0>}]
}
constraints [
 {bone1Name=dyn_sword_r, bone2Name=pack, swing1LimitDegrees=12, swing2LimitDegrees=15, twistLimitDegrees=8}
 {bone1Name=dyn_sword_l, bone2Name=pack, swing1LimitDegrees=12, swing2LimitDegrees=15, twistLimitDegrees=8}
 {bone1Name=dyn_head, bone2Name=pack, swing1LimitDegrees=30, swing2LimitDegrees=8, twistLimitDegrees=45, twistMotion=Locked}
]
